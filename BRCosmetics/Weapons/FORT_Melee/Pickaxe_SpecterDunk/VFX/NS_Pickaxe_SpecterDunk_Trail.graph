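# Niagara system that draws the swing trail for the SpecterDunk pickaxe cosmetic using three ribbon emitters on anim-trail bone/socket pairs.
# FillTrail spawns two adjoining trails, while TipTrail and ColorTrail share another pair.
# User.VariantIndex selects each emitter's color or color curve, which is tinted by time of day.
# Curves fade alpha, scale ribbon width and drive a dynamic material parameter over each trail's life.
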NiagaraSystem /BRCosmetics/Weapons/FORT_Melee/Pickaxe_SpecterDunk/VFX/NS_Pickaxe_SpecterDunk_Trail
dependencies {
 usedDataInterfaces [&NiagaraDataInterfaceColorCurve, &NiagaraDataInterfaceCurve, &NiagaraDataInterfaceSocketReader]
 usedDynamicInputs []
 usedModules [
  &self:SystemSpawnScript.SystemScriptSource.SystemScriptGraph.NiagaraNodeAssignment_4.SetVariables_#1
  &SystemState
  &self:SystemSpawnScript.SystemScriptSource.SystemScriptGraph.NiagaraNodeAssignment_1.SetVariables_#2
 ]
}
emitters {
 ColorTrail {
  emitterSpawnScript [&FN_RibbonIDOffsetUtility()]
  emitterUpdateScript [
   &EmitterState()
   &FN_SpawnAnimTrail()
  ]
  particleSpawnScript [
   &FN_InitializeAnimTrailBoneIndecies(First Bone or Socket Index=2, Second Bone or Socket Index=3)
   &FN_InitializeAnimTrail(
    Color=&MergeHelper_ColorSwitch(
     0=<rgba 0.4262 1 0 0.5>
     1=<rgba 1.5 1.0274 0 0.5>
     2=<rgba 5 0.01 2.6136 0.5>
     3=<rgba 5 0.01 2.6136 0.5>
     Index=&MakeIntFromFloat(Float=$User.VariantIndex)
     Num=::NewEnumerator2
    )
    Lifetime=0.4
   )
   &Color(Color=&MultiplyLinearColor(
     Linear Color=$Particles.Initial.Color
     Scale Factor=&MakeVector4_FromLinearColor(LinearColor=&FNTimeOfDayColor(
       Night Color=<rgba 0.7461 0.6737 1 1>
       Morning Color=<rgba 0.8728 0.9258 1 1>
       Day Color=<rgba 1 1 1 1>
       Evening Color=<rgba 0.8728 0.9258 1 1>
      ))
    ))
  ]
  particleUpdateScript [
   &SolveForcesAndVelocity(Clamp Velocity=false, Limit Acceleration=false)
   &FN_UpdateAnimTrail(Locked Particle RibbonWidth=1) {module=FN_UpdateAnimTrail001}
   &ScaleColor(
    Scale Mode=::NewEnumerator1
    Scale RGBA=&MakeVector4_FromLinearColor(LinearColor=$System.ToDColor)
    ScaleRGBA=true
   ) {module=ScaleColor001}
   &DynamicMaterialParameters(
    Index 0 Param 1=&FloatFromCurve(FloatCurve=DataInterface() {propertyValues {Curve {keys [<at 0 1>,<at 1 0>]}}})
    Index 0 Param 2=0.3
    Index 0 Param 3=2
    Index 0 Param 4=1
    Module.Index0Param0Write=true
    Module.Index0Param1Write=true
    Module.Index0Param2Write=true
    Module.Index0Param3Write=true
   )
   &ScaleRibbonWidth(Ribbon Width Scale=&FloatFromCurve(FloatCurve=DataInterface() {propertyValues {Curve {keys [{interpMode=RCIM_Cubic, leaveTangent=-7.4157, tangentMode=RCTM_Break, time=0.2, value=1}, {tangentMode=RCTM_User, time=0.45, value=0}]}}}))
   &ScaleColor(
    Scale Alpha=&FloatFromCurve(
     FloatCurve=DataInterface() {propertyValues {Curve {keys [
         {interpMode=RCIM_Cubic, leaveTangent=-19.4136, tangentMode=RCTM_Break, time=0, value=1}
         {arriveTangent=-0.1476, tangentMode=RCTM_User, time=0.13, value=0}
        ]}}}
     Scale Curve=0.35
    )
    Scale RGB=&MakeVectorFromLinearColorRGB(Color=&ColorFromCurve(ColorCurve=DataInterface() {propertyValues {
        AlphaCurve {keys []}
        BlueCurve {keys [<at 0.0167 4 RCIM_Cubic>,<at 0.2127 1>,<at 0.5157 0.75 RCIM_Cubic>]}
        GreenCurve {keys [<at 0.0167 4 RCIM_Cubic>,<at 0.2127 1>,<at 0.5157 0.75 RCIM_Cubic>]}
        RedCurve {keys [<at 0.0167 4 RCIM_Cubic>,<at 0.2127 1>,<at 0.5157 0.75 RCIM_Cubic>]}
       }}))
    ScaleA=true
    ScaleRGB=true
   )
  ]
  rendererClasses [&NiagaraRibbonRendererProperties]
  renderers [{rendererClass=&NiagaraRibbonRendererProperties}]
 }
 FillTrail {
  emitterSpawnScript [&FN_RibbonIDOffsetUtility()]
  emitterUpdateScript [
   &EmitterState()
   &FN_SpawnAnimTrail(Number of Anim Trails=2)
  ]
  particleSpawnScript [
   &FN_InitializeAnimTrailBoneIndecies(First Bone or Socket Index=1, Second Bone or Socket Index=0)
   &FN_InitializeAnimTrailBoneIndecies(AnimTrail Selection=1, First Bone or Socket Index=1, Second Bone or Socket Index=2) {module=FN_InitializeAnimTrailBoneIndecies001}
   &FN_InitializeAnimTrail(Lifetime=0.3)
  ]
  particleUpdateScript [
   &SolveForcesAndVelocity(Clamp Velocity=false, Limit Acceleration=false)
   &FN_UpdateAnimTrail(Locked Particle RibbonWidth=1) {module=FN_UpdateAnimTrail001}
   &ScaleColor(
    Scale Alpha=&FloatFromCurve(
     FloatCurve=DataInterface() {propertyValues {Curve {keys [
         {interpMode=RCIM_Cubic, leaveTangent=-18.1448, tangentMode=RCTM_Break, time=0, value=1}
         {arriveTangent=-0.1476, tangentMode=RCTM_User, time=0.13, value=0}
        ]}}}
     Scale Curve=0.75
    )
    Scale RGB=&MakeVectorFromLinearColorRGB(Color=&MultiplyLinearColor(
      Linear Color=&MergeHelper_ColorSwitch(
       0=<rgba 0.2343 0.381 0.4531 1.2>
       1=&ColorFromCurve(ColorCurve=DataInterface() {propertyValues {
          AlphaCurve {keys [<at 0 1 RCIM_Cubic>]}
          BlueCurve {keys [<at 0 0>,<at 0.0241 0.0317 RCIM_Cubic>,<at 0.0835 0 RCIM_Cubic>]}
          GreenCurve {keys [<at 0 0.4954>, {interpMode=RCIM_Cubic, leaveTangent=-4.2403, time=0.0241, value=0.379}, <at 0.0835 0.0333 RCIM_Cubic>]}
          RedCurve {keys [<at 0 1>,<at 0.0241 1 RCIM_Cubic>,<at 0.0835 1 RCIM_Cubic>]}
         }})
       2=&ColorFromCurve(ColorCurve=DataInterface() {propertyValues {
          AlphaCurve {keys [<at 0 1 RCIM_Cubic>]}
          BlueCurve {keys [<at 0 0.9309>, {arriveTangent=0.9972, interpMode=RCIM_Cubic, leaveTangent=0.9972, time=0.0127, value=0.6458}, <at 0.0693 1 RCIM_Cubic>]}
          GreenCurve {keys [<at 0 0>,<at 0.0127 0.1278 RCIM_Cubic>,<at 0.0693 0 RCIM_Cubic>]}
          RedCurve {keys [<at 0 2>, {arriveTangent=-25.7649, interpMode=RCIM_Cubic, leaveTangent=-25.7649, time=0.0127, value=0.4611}, <at 0.0693 0.2135 RCIM_Cubic>]}
         }})
       3=&ColorFromCurve(ColorCurve=DataInterface() {propertyValues {
          AlphaCurve {keys [<at 0 1 RCIM_Cubic>]}
          BlueCurve {keys [<at 0 2>, {interpMode=RCIM_Cubic, leaveTangent=-14.4219, time=0.0127, value=0.6458}, <at 0.0454 1 RCIM_Cubic>]}
          GreenCurve {keys [<at 0 1.2985>, {interpMode=RCIM_Cubic, leaveTangent=-16.7554, time=0.0127, value=0.3145}, <at 0.0454 0.1367 RCIM_Cubic>]}
          RedCurve {keys [<at 0 0.4619>, {interpMode=RCIM_Cubic, leaveTangent=-2.2465, time=0.0127, value=0.5277}, <at 0.0454 0.3061 RCIM_Cubic>]}
         }})
       Index=&MakeIntFromFloat(Float=$User.VariantIndex)
       Num=::NewEnumerator2
      )
      Scale Factor=&MakeVector4_FromLinearColor(LinearColor=$System.ToDColor)
     ))
    ScaleA=true
    ScaleRGB=true
   )
   &ScaleRibbonWidth(Ribbon Width Scale=&FloatFromCurve(FloatCurve=DataInterface() {propertyValues {Curve {keys [{interpMode=RCIM_Cubic, leaveTangent=0.55, tangentMode=RCTM_Break, time=0.2, value=1}, <at 0.45 0.1>]}}}))
   &DynamicMaterialParameters(
    Index 0 Param 1=&FloatFromCurve(FloatCurve=DataInterface() {propertyValues {Curve {keys [<at 0 1>,<at 1 0>]}}})
    Index 0 Param 2=0.3
    Index 0 Param 3=2
    Index 0 Param 4=1
    Module.Index0Param0Write=true
    Module.Index0Param1Write=true
    Module.Index0Param2Write=true
    Module.Index0Param3Write=true
   )
  ]
  rendererClasses [&NiagaraRibbonRendererProperties]
  renderers [{rendererClass=&NiagaraRibbonRendererProperties}]
 }
 TipTrail {
  emitterSpawnScript [&FN_RibbonIDOffsetUtility()]
  emitterUpdateScript [
   &EmitterState()
   &FN_SpawnAnimTrail()
  ]
  particleSpawnScript [
   &FN_InitializeAnimTrailBoneIndecies(First Bone or Socket Index=2, Second Bone or Socket Index=3)
   &FN_InitializeAnimTrail(
    AnimTrailWidth=2
    Color=&MergeHelper_ColorSwitch(
     0=<rgba 0.4807 0.6922 0.75 0.5>
     1=<rgba 0.9304 0.3002 0.013 0.5>
     2=<rgba 0.9304 0.013 0.3002 0.5>
     3=<rgba 0.4807 0.6922 0.75 0.5>
     Index=&MakeIntFromFloat(Float=$User.VariantIndex)
     Num=::NewEnumerator2
    )
   )
   &Color(Color=&MultiplyLinearColor(
     Linear Color=$Particles.Initial.Color
     Scale Factor=&MakeVector4_FromLinearColor(LinearColor=&FNTimeOfDayColor(
       Night Color=<rgba 0.5387 0.487 1 1>
       Morning Color=<rgba 0.8728 0.9258 1 1>
       Day Color=<rgba 1 1 1 1>
       Evening Color=<rgba 0.8728 0.9258 1 1>
      ))
    ))
  ]
  particleUpdateScript [
   &SolveForcesAndVelocity(Clamp Velocity=false, Limit Acceleration=false)
   &FN_UpdateAnimTrail(Locked Particle RibbonWidth=1) {module=FN_UpdateAnimTrail001}
   &ScaleColor(
    Scale Alpha=&FloatFromCurve(
     FloatCurve=DataInterface() {propertyValues {Curve {keys [
         {interpMode=RCIM_Cubic, leaveTangent=-19.5372, tangentMode=RCTM_Break, time=0, value=1}
         {arriveTangent=-0.1476, tangentMode=RCTM_User, time=0.13, value=0}
        ]}}}
     Scale Curve=0.35
    )
    Scale RGB=&MakeVectorFromLinearColorRGB(Color=$System.ToDColor)
    ScaleA=true
    ScaleRGB=true
   )
   &ScaleRibbonWidth(Ribbon Width Scale=&FloatFromCurve(FloatCurve=DataInterface() {propertyValues {Curve {keys [{interpMode=RCIM_Cubic, leaveTangent=-5.18, tangentMode=RCTM_Break, time=0.2, value=1}, <at 0.65 0.1>]}}}))
   &DynamicMaterialParameters(
    Index 0 Param 1=&FloatFromCurve(FloatCurve=DataInterface() {propertyValues {Curve {keys [<at 0 1>,<at 1 0>]}}})
    Index 0 Param 2=0.3
    Index 0 Param 3=2
    Index 0 Param 4=1
    Module.Index0Param0Write=true
    Module.Index0Param1Write=true
    Module.Index0Param2Write=true
    Module.Index0Param3Write=true
   )
  ]
  rendererClasses [&NiagaraRibbonRendererProperties]
  renderers [{rendererClass=&NiagaraRibbonRendererProperties}]
 }
}
summary {
 systemName=NS_Pickaxe_SpecterDunk_Trail
 emitters [
  {emitterName=FillTrail, rendererClasses [&NiagaraRibbonRendererProperties]}
  {emitterName=TipTrail, rendererClasses [&NiagaraRibbonRendererProperties]}
  {emitterName=ColorTrail, rendererClasses [&NiagaraRibbonRendererProperties]}
 ]
 userVariables [
  User.RibbonIDOffset {type=NiagaraInt32}
  User.VariantIndex {type=NiagaraFloat, defaultValue=0, description=""}
 ]
}
